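; Gameplay ability for the Season 18 Dark Jonesy Oracle touch-cube quest, active only while that objective is in progress.
; It swaps in a temporary cosmetic with looping audio and backpack VFX, then a repeating timer ramps the backpack's dynamic-material emissive and Niagara glow.
; A parallel delay sends the quest stat event using target tags.
; A timer ends the ability, reverting the cosmetic, playing removal effects and stopping audio.
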(event Custom|ChargeUpFX
  (bind _chargetime (Variables|Default|GetChargeTime))
  (if (>= (Variables|Default|GetChargeTimeCurrent) (Variables|Default|GetChargeTime))
    (Utilities|Time|ClearandInvalidateTimerbyHandle (Variables|Default|GetChargeTimer))
    (Variables|Default|SetChargeTimeCurrent 0.0)
    (else
      (bind _output_get (Variables|Default|SetChargeTimeCurrent (+ (Variables|Default|GetChargeTimeCurrent) (Variables|Default|GetChargeTimeIncrement))))
      (Rendering|Material|SetScalarParameterValue (Variables|Default|GetREF_BackpackEmissive) "emissive mult" (Math|Float|FInterpEaseinOut 0.0 1e+01 (Math|Float|MapRangeClamped _output_get 0.0 _chargetime 0.0 1.0) 1.0) :!class /Script/Engine.MaterialInstanceDynamic)
      (Effects|Components|ParticleSystem|SetFloatParameter (Variables|Default|GetFXComponent) "GlowScale" (Math|Float|MapRangeClamped _output_get 0.0 _chargetime 0.0 1.0)))))

(event Ability|EventActivateAbility
  (bind _returnvalue (Ability|GetAvatarActorfromActorInfo))
  (bind _returnvalue_2 (Fortnite|GetFortPlayerControllerfromActor _returnvalue))
  (bind _returnvalue_3 (Quests|GetQuestManager _returnvalue_2 "Invalid"))
  (Utilities|IsValid _returnvalue_3
    (:Is%20Valid
      (if (Quest|IsObjectivewithNameinProgress _returnvalue_3 "/BattlePassS18/Items/QuestItems/Transient/quest_s18_transient_darkjonesy_oracle_q03.quest_s18_transient_darkjonesy_oracle_q03" "quest_s18_transient_darkjonesy_oracle_q03_obj0" :!class /Script/FortniteGame.FortQuestManager)
        (CallFunction|ToggleCosmetics "Active")
        (Utilities|Time|SetTimerbyEvent (EventDispatchers|CreateEvent :!event K2_EndAbility) 1e+01)
        (Utilities|FlowControl|Sequence
          (:then_0
            (Utilities|FlowControl|Delay 1.5)
            (bind _returnvalue_4 (Utilities|Time|SetTimerbyEvent (EventDispatchers|CreateEvent :!event Custom|ChargeUpFX) 0.1 true))
            (Variables|Default|SetChargeTimer _returnvalue_4)
            (bind _asfort_player_pawn (Utilities|Casting|CastToFortPlayerPawn (Ability|GetAvatarActorfromActorInfo)))
            (bind _returnvalue_5 (Niagara|SpawnSystemAttached (Variables|Default|GetChargingVFX) (Customization|GetSkeletalMeshforPartType _asfort_player_pawn "Backpack" :!class /Script/FortniteGame.FortPlayerPawn) "root" "0, 0, 0" "0, 0, 0" "SnapToTarget" true))
            (Variables|Default|SetFXComponent _returnvalue_5)
            (bind _returnvalue_6 (Rendering|Material|CreateDynamicMaterialInstance (Variables|Default|GetREF_BackpackMat) "newInstance" "None" :!class /Script/Engine.KismetMaterialLibrary))
            (bind _output_get (Variables|Default|SetREF_BackpackEmissive _returnvalue_6))
            (Rendering|Material|SetMaterial (Customization|GetSkeletalMeshforPartType _asfort_player_pawn "Backpack" :!class /Script/FortniteGame.FortPlayerPawn) 0 _output_get))
          (:then_1
            ;[1 0 0] Needs to be cleaned up a bit, but submitting for now.
            (Utilities|FlowControl|Delay 5.5)
            (Quest|SendComplexCustomStatEvent _returnvalue_3 0 "(GameplayTags=)" (Variables|Default|GetTarget%20Tags) 1)))))))

(event Custom|ToggleCosmetics
  (params (ForcedSwapState :type EFortAppliedSwapItemAndVariantState%20Enum))
  (bind _audiocomponent (Variables|Default|GetAudioComponent))
  (bind _returnvalue (GUID|GenerateGuidfromObject (Variables|Default|GetQuestID)))
  (bind _asfort_player_pawn (Utilities|Casting|CastToFortPlayerPawn (Ability|GetAvatarActorfromActorInfo self)))
  (Class|FortPlayerPawn|SwapPawnCosmeticItemandVariant _asfort_player_pawn _returnvalue (Variables|Default|GetPlayer%20Cosmetic%20Overrides) false ForcedSwapState false)
  (if (Utilities|Enum|Equal(Enum) ForcedSwapState "Active")
    (bind _asfort_player_pawn_2 (Utilities|Casting|CastToFortPlayerPawn (Ability|GetAvatarActorfromActorInfo)))
    (bind _returnvalue_2 (Audio|SpawnSoundAttached (Variables|Default|GetLoopingSound) (Variables|/Script/Engine.Actor|GetRootComponent _asfort_player_pawn_2)))
    (Variables|Default|SetAudioComponent _returnvalue_2)
    (Utilities|FlowControl|Delay 0.15) ; Slight Delay to appear when Backpack appears
    (Niagara|SpawnSystemAttached (Variables|Default|GetApplicationVFX) (Customization|GetSkeletalMeshforPartType _asfort_player_pawn_2 "Backpack" :!class /Script/FortniteGame.FortPlayerPawn) "root" "0, 0, 0" "0, 0, 0" "SnapToTarget" true)
    (else
      (Niagara|SpawnSystemAttached (Variables|Default|GetRemovalVFX) (Customization|GetSkeletalMeshforPartType _asfort_player_pawn_2 "Backpack" :!class /Script/FortniteGame.FortPlayerPawn) "root" "0, 0, 0" "0, 0, 0" "SnapToTarget" true)
      (Audio|Components|Audio|Stop _audiocomponent)
      (Audio|SpawnSoundAttached "/BattlePassS18/QuestActors/Assets/Sound/Quest_Oracle_03_Vessel_Loop_Stop.Quest_Oracle_03_Vessel_Loop_Stop" _audiocomponent)
      (Utilities|FlowControl|Delay 4.0)
      (Components|DestroyComponent _audiocomponent))))

(event Ability|EventOnEndAbility
  (params (bWasCancelled :type Boolean))
  (CallFunction|ToggleCosmetics "Inactive")
  (Components|DestroyComponent (Variables|Default|GetFXComponent)))
